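; Deprecated HeatSnake weapon child Blueprint, a chain-whip-style melee weapon that plays an equip GameplayCue and stores target and tether point data for getters.
; On cosmetic-capable clients, async targeting caches the last hit and sends its distance to the anim Blueprint to scale chain length, with maximum range as fallback.
; Targeting component updates feed the weapon's target-state component.
; A custom VirtualTick draws a forward debug box trace.
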(event EventBeginPlay
  (|CallParentFunction)
  (Networking|CanExecuteCosmeticEvents))

(event FortWeapon|EventOnEquippedAndAttached
  ;[0.15 0.01265 0 0.5] Execute equip screen FX
  (|CallParentFunction)
  (GameplayCue|ExecuteGameplayCueOnActor(Burst) (Game|GetPlayerPawn 0) (Variables|Default|GetEquipGameplayCueTag) (Ability|GameplayCue|MakeGameplayCueParameters 0.0)))

(event Custom|VirtualTick
  (Collision|BoxTraceByChannel :Start (Transformation|GetActorLocation) :End (+ (Transformation|GetActorLocation) (* (Transformation|GetActorForwardVector) 6.5e+03)) :HalfSize "60.000000,60.000000,60.000000" :Orientation (Transformation|GetActorRotation) :TraceChannel "TraceTypeQuery1" :bTraceComplex false :DrawDebugType "ForOneFrame" :bIgnoreSelf true :TraceColor "(R=1.000000,G=0.000000,B=0.000000,A=1.000000)" :TraceHitColor "(R=0.000000,G=1.000000,B=0.000000,A=1.000000)" :DrawTime 5.0))

(event EventUpdateAttackRangeAnimBP
  (bind _returnvalue (GameInstanceSubsystems|GetTargetingSubsystem))
  ;[0.09147 0.03613 0.1094] UpdateAttackRangeAnimBP: Perform a targeting trace to figure out the distance of the trace in the aiming direction.
  ; This value is used in the animation blueprint to scale the length of the chain on the attack.
  ; Note: This logic needs to run on simulated and autonomous proxies.
  (Networking|CanExecuteCosmeticEvents
    (:True
      (bind _b_chainwhiptargetingcomponent (Variables|Default|GetB_ChainWhipTargetingComponent)
        (:then
          (Utilities|IsValid (General|GetPlayerPawn _b_chainwhiptargetingcomponent :!class /SunRoseChainWhipGameplay/Items/ChainWhip/B_ChainWhipTargetingComponent.B_ChainWhipTargetingComponent_C)
            (:Is%20Valid
              (bind (_asynctaskproxy _targetinghandle) (Class|AsyncActionPerformTargeting|PerformTargetingAsyncAction (General|GetPlayerPawn _b_chainwhiptargetingcomponent :!class /SunRoseChainWhipGameplay/Items/ChainWhip/B_ChainWhipTargetingComponent.B_ChainWhipTargetingComponent_C) "/HeatSnake/Gameplay/Targeting/TP_HeatSnake_MeleeTargeting.TP_HeatSnake_MeleeTargeting")
                (:Targeted
                  (Utilities|IsValid _returnvalue
                    (:Is%20Valid
                      (if (Utilities|Array|IsValidIndex (TargetingSystem|TargetingResults|GetTargetingResults _returnvalue _targetinghandle) (Utilities|Array|LastIndex (TargetingSystem|TargetingResults|GetTargetingResults _returnvalue _targetinghandle)))
                        (Variables|Default|SetLastTargetData (Utilities|Array|Get(aref) (TargetingSystem|TargetingResults|GetTargetingResults _returnvalue _targetinghandle) (Utilities|Array|LastIndex (TargetingSystem|TargetingResults|GetTargetingResults _returnvalue _targetinghandle))))
                        (if (Collision|BreakHitResult (Utilities|Array|Get(aref) (TargetingSystem|TargetingResults|GetTargetingResults _returnvalue _targetinghandle) (Utilities|Array|LastIndex (TargetingSystem|TargetingResults|GetTargetingResults _returnvalue _targetinghandle))))
                          ;;[0.03005 0.151 0.1253] Update the attack range based on the distance result of the trace.
                          (bind _animbp (Default|GetWeaponAnimBP)
                            (:Valid
                              (Class|ChainsSunRoseSharpWepAnimBP|UpdateAttackRange _animbp (.Distance (Collision|BreakHitResult (Utilities|Array|Get(aref) (TargetingSystem|TargetingResults|GetTargetingResults _returnvalue _targetinghandle) (Utilities|Array|LastIndex (TargetingSystem|TargetingResults|GetTargetingResults _returnvalue _targetinghandle))))))))
                          (else
                            _))
                        (else
                          _))
                      ;;[0.06385 0.194 0.1979] Any failure to perform the trace should update the attack range to max distance.
                      (bind _animbp_2 (Default|GetWeaponAnimBP)
                        (:Valid
                          (Class|ChainsSunRoseSharpWepAnimBP|UpdateAttackRange _animbp_2 (ScalableFloat|GetValueatLevel (Variables|Hotfix|GetHotfix_MeleeAttackRange))))))
                    (:Is%20Not%20Valid
                      _)))))
            (:Is%20Not%20Valid
              _)))
        (:else
          _)))))

(event HeatSnakeWeapon|EventSetTetherPointData
  (params (NewTetherPoint :type Vector :default "0, 0, 0"))
  (Variables|Default|SetTetherPoint NewTetherPoint))

(event OnTargetInfoUpdated(B_HeatSnakeTargetingComponent)
  (params (HitResult :type Hit%20Result%20Structure) (TargetType :type ChainWhipTargetActorTypeEnum%20Enum :default "NewEnumerator0"))
  (bind (_bblockinghit _binitialoverlap _time _distance _location _impactpoint) (Collision|BreakHitResult HitResult))
  (Class|FortWeaponComponentTargetState|UpdateTargetInfo (Variables|Default|GetFortWeaponComponent_TargetState) (or (Utilities|Enum|Equal(Enum) TargetType "NewEnumerator1") (Utilities|Enum|Equal(Enum) TargetType "NewEnumerator2")) _impactpoint))

(fn GetTargetData
  (return (Variables|Default|GetLastTargetData)))

(fn GetTetherPointData
  (return (Variables|Default|GetTetherPoint)))

(fn ConstructionScript
  (|CallParentFunction))
